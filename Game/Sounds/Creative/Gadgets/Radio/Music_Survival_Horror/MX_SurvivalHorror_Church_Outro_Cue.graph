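# Sound cue for the survival horror church outro music track of the Creative radio gadget.
# A single wave player feeds the output directly with one wav, with no randomization, looping node or mixing.
# It routes to the SFX sound class and uses the shared Creative scoring attenuation with a very large max distance.
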SoundCue /Game/Sounds/Creative/Gadgets/Radio/Music_Survival_Horror/MX_SurvivalHorror_Church_Outro_Cue
root=WavePlayer_0
nodes [{
  id=WavePlayer_0
  properties {soundWaveAssetPtr=/Game/Sounds/Creative/Gadgets/Radio/Music_Survival_Horror/Wavs/MX_SurvivalHorror_Church_Outro_01}
 }]
properties {
 attenuationSettings=/Game/Sounds/Creative/Attenuations/Fort_Creative_Scoring_Att
 maxDistance=2097152
 soundClassObject=&SFX
 volumeMultiplier=1
}
