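# Sound cue for a looping evil-fairground survival horror music track, filed under the Creative radio gadget music.
# A single looping wave player feeds the root directly and is assigned to the base music sound class.
# An attenuation override turns off spatialization and sets a very long falloff with a light reverb send.
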SoundCue /Game/Sounds/Creative/Gadgets/Radio/Music_Survival_Horror/MX_SurvivalHorror_Fairground_Evil_Loop_Cue
root=WavePlayer_0
nodes [{
  id=WavePlayer_0
  properties {bLooping=true, soundWaveAssetPtr=/Game/Sounds/Creative/Gadgets/Radio/Music_Survival_Horror/Wavs/MX_SurvivalHorror_Fairground_Evil_Loop_01}
 }]
properties {
 bOverrideAttenuation=true
 maxDistance=100000
 soundClassObject=/Game/Sounds/FortSoundClassesAndModes/Fort_11_Base_Music.Fort_11_Base_music
 volumeMultiplier=1
 attenuationOverrides {bSpatialize=false, falloffDistance=40000, reverbWetLevelMax=0.2, reverbWetLevelMin=0.1, attenuationShapeExtents {x=60000}}
}
