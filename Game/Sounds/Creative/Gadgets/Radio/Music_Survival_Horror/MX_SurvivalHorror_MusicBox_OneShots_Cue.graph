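# Sound cue for survival horror music box one-shots, filed under the Creative radio gadget's music sounds.
# A root Random node picks among five equally weighted wave players, each holding its own one-shot variation.
# The cue is routed to the SFX sound class rather than a music class and uses the shared Creative scoring attenuation.
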SoundCue /Game/Sounds/Creative/Gadgets/Radio/Music_Survival_Horror/MX_SurvivalHorror_MusicBox_OneShots_Cue
root=Random_0
nodes [
 {id=Random_0, children [WavePlayer_1,WavePlayer_2,WavePlayer_3,WavePlayer_4,WavePlayer_5], properties {weights [1,1,1,1,1]}}
 {
  id=WavePlayer_1
  properties {soundWaveAssetPtr=/Game/Sounds/Creative/Gadgets/Radio/Music_Survival_Horror/Wavs/MX_SurvivalHorror_MusicBox_OneShot_01}
 }
 {
  id=WavePlayer_2
  properties {soundWaveAssetPtr=/Game/Sounds/Creative/Gadgets/Radio/Music_Survival_Horror/Wavs/MX_SurvivalHorror_MusicBox_OneShot_02}
 }
 {
  id=WavePlayer_3
  properties {soundWaveAssetPtr=/Game/Sounds/Creative/Gadgets/Radio/Music_Survival_Horror/Wavs/MX_SurvivalHorror_MusicBox_OneShot_03}
 }
 {
  id=WavePlayer_4
  properties {soundWaveAssetPtr=/Game/Sounds/Creative/Gadgets/Radio/Music_Survival_Horror/Wavs/MX_SurvivalHorror_MusicBox_OneShot_04}
 }
 {
  id=WavePlayer_5
  properties {soundWaveAssetPtr=/Game/Sounds/Creative/Gadgets/Radio/Music_Survival_Horror/Wavs/MX_SurvivalHorror_MusicBox_OneShot_05}
 }
]
properties {
 attenuationSettings=/Game/Sounds/Creative/Attenuations/Fort_Creative_Scoring_Att
 maxDistance=2097152
 soundClassObject=&SFX
 volumeMultiplier=1
}
